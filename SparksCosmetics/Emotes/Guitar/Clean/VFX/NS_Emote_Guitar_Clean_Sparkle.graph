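# Niagara sprite effect for the clean guitar emote in SparksCosmetics, producing a warm golden flare with twinkling sparkles.
# Flare bursts counter-rotating, fading sprites through three renderers, with spawn group choosing a flickering or growing size curve and driving a visibility tag.
# Sparkles combines a burst and steady spawn rate in a box, with a spawn-group-masked loop age controlling box height and lifetime as sprites pulse and spin.
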NiagaraSystem /SparksCosmetics/Emotes/Guitar/Clean/VFX/NS_Emote_Guitar_Clean_Sparkle
dependencies {
 usedDataInterfaces [&NiagaraDataInterfaceCurve, &NiagaraDataInterfaceArrayFloat, &NiagaraDataInterfaceColorCurve]
 usedDynamicInputs []
 usedModules [&SystemState]
}
emitters {
 Flare {
  emitterUpdateScript [
   &EmitterState()
   &SpawnBurst_Instantaneous(Spawn Group=1, Spawn Time=0.1, Use Spawn Probability=false, UseLoopCountLimit=false) {module=SpawnBurst_Instantaneous001}
   &SpawnBurst_Instantaneous(Spawn Count=2, Spawn Time=0.1, Use Spawn Probability=false, UseLoopCountLimit=false)
  ]
  particleSpawnScript [
   &InitializeParticle(
    Color=&RandomRangeLinearColor(Minimum=<rgba 0.25 0.2013 0.0857 1>, Maximum=<rgba 0.35 0.2887 0.1431 1>)
    Lifetime Max=0.8
    Lifetime Min=0.5
    Lifetime Mode=::NewEnumerator1
    Sprite Rotation Mode=::NewEnumerator2
    Sprite Size Mode=::NewEnumerator3
    Uniform Sprite Size=&SelectFloatFromArray(
     Direct Array Index=&ReturnExecIndex()
     Float Selection Array=DataInterface() {propertyValues {FloatData [50,100]}}
    )
    UsePositionOffset=true
    Write Material Random=true
   )
   /SetVariables(Particles.VisibilityTag=&SpawnGroup()) {isSetParameters=true}
   &CameraOffset(Camera Offset Amount=-1)
  ]
  particleUpdateScript [
   &ParticleState()
   &ScaleSpriteSize(
    Scale Sprite Size Mode=::NewEnumerator0
    Uniform Scale Factor=&Lerp_Float(
     A=&FloatFromCurve(FloatCurve=DataInterface() {propertyValues {Curve {keys [
          {arriveTangent=7.2727, interpMode=RCIM_Cubic, leaveTangent=7.2727, tangentMode=RCTM_User, time=0, value=0}
          {arriveTangent=0.8575, interpMode=RCIM_Cubic, leaveTangent=0.8575, tangentMode=RCTM_User, time=0.2073, value=0.9}
          {arriveTangent=-3.3258, interpMode=RCIM_Cubic, leaveTangent=-3.3258, time=0.2422, value=0.714}
          {arriveTangent=-2.3265, interpMode=RCIM_Cubic, leaveTangent=-2.3265, time=0.316, value=0.5387}
          {arriveTangent=-0.4054, interpMode=RCIM_Cubic, leaveTangent=-0.4054, time=0.3894, value=0.3714}
          {arriveTangent=1.8019, interpMode=RCIM_Cubic, leaveTangent=1.8019, time=0.4302, value=0.4924}
          {interpMode=RCIM_Cubic, tangentMode=RCTM_User, time=0.5179, value=0.6029}
          {arriveTangent=-2.1399, interpMode=RCIM_Cubic, leaveTangent=-2.1399, time=0.56, value=0.4652}
          {arriveTangent=-2.2672, interpMode=RCIM_Cubic, leaveTangent=-2.2672, time=0.6389, value=0.344}
          {arriveTangent=-0.7161, interpMode=RCIM_Cubic, leaveTangent=-0.7161, time=0.7, value=0.1479}
          {arriveTangent=1.421, interpMode=RCIM_Cubic, leaveTangent=1.421, time=0.7311, value=0.2779}
          {arriveTangent=-1.2231, interpMode=RCIM_Cubic, leaveTangent=-1.2231, time=0.8109, value=0.3054}
          {arriveTangent=-1.6149, interpMode=RCIM_Cubic, leaveTangent=-1.6149, time=0.8479, value=0.1351}
          {tangentMode=RCTM_User, time=1, value=0}
         ]}}})
     B=&FloatFromCurve(
      FloatCurve=DataInterface() {propertyValues {Curve {keys [
          {arriveTangent=0.6922, interpMode=RCIM_Cubic, leaveTangent=0.6922, tangentMode=RCTM_User, time=0, value=0.3}
          {arriveTangent=0.8174, tangentMode=RCTM_User, time=1, value=1}
         ]}}}
      Scale Curve=2
     )
     Alpha=&MakeFloatFromInt(Int=$Particles.VisibilityTag)
    )
   )
   &SpriteRotationRate(Rotation Rate=&Multiply_Float(
     A=&FloatFromCurve(FloatCurve=DataInterface() {propertyValues {Curve {keys [{arriveTangent=-2.6251, interpMode=RCIM_Cubic, leaveTangent=-2.6251, tangentMode=RCTM_User, time=0, value=1}, <at 1 0>]}}})
     B=&SelectFloatFromArray(
      Direct Array Index=&ReturnExecIndex()
      Float Selection Array=DataInterface() {propertyValues {FloatData [-400,400]}}
     )
    ))
   &ScaleColor(
    Scale Alpha=&FloatFromCurve(FloatCurve=DataInterface() {propertyValues {Curve {keys [
         {arriveTangent=-2.8083, interpMode=RCIM_Cubic, leaveTangent=-2.8083, tangentMode=RCTM_User, time=0, value=1}
         {arriveTangent=-1, interpMode=RCIM_Cubic, leaveTangent=-1, time=0.3, value=0.3}
         <at 1 0 RCIM_Cubic>
        ]}}})
    ScaleA=true
    ScaleRGB=true
   )
   /SetVariables(Particles.CoronaColor:LinearColor=&ColorFromCurve(ColorCurve=DataInterface() {propertyValues {
       AlphaCurve {keys [
         {arriveTangent=-3.4285, interpMode=RCIM_Cubic, leaveTangent=-3.4285, tangentMode=RCTM_User, time=0, value=1}
         {arriveTangent=-0.3627, interpMode=RCIM_Cubic, leaveTangent=-0.3627, tangentMode=RCTM_User, time=1, value=0}
        ]}
       BlueCurve {keys [<at 0 1 RCIM_Cubic>,<at 1 1 RCIM_Cubic>]}
       GreenCurve {keys [<at 0 1 RCIM_Cubic>,<at 1 1 RCIM_Cubic>]}
       RedCurve {keys [<at 0 1 RCIM_Cubic>,<at 1 1 RCIM_Cubic>]}
      }})) {isSetParameters=true}
   /SetVariables(Particles.CoronaRotation:NiagaraFloat=&Multiply_Float(A=$Particles.SpriteRotation, B=0.25)) {isSetParameters=true}
   &DynamicMaterialParameters(
    Index 0 Param 1=&FloatFromCurve(FloatCurve=DataInterface() {propertyValues {Curve {keys [<at 0 0.5>,<at 1 1>]}}})
    Index 0 Param 2=1
    Index 0 Param 3=1
    Index 0 Param 4=1
    Module.Index0Param0Write=true
    Module.Index0Param1Write=true
    Module.Index0Param2Write=true
    Module.Index0Param3Write=true
   )
  ]
  rendererClasses [&NiagaraSpriteRendererProperties]
  renderers [
   {rendererClass=&NiagaraSpriteRendererProperties}
   {rendererClass=&NiagaraSpriteRendererProperties, rendererIndex=1}
   {rendererClass=&NiagaraSpriteRendererProperties, rendererIndex=2}
  ]
 }
 Sparkles {
  emitterUpdateScript [
   &EmitterState()
   &SpawnBurst_Instantaneous(Spawn Count=3, Spawn Group=1, Use Spawn Probability=false, UseLoopCountLimit=false)
   &SpawnRate(SpawnRate=5, Use Spawn Probability=false)
  ]
  particleSpawnScript [
   /SetVariables(Transient.Spread:NiagaraFloat=&MaskFloatBySpawnGroup(Masked Float=$Emitter.NormalizedLoopAge, Passthrough Float=1)) {isSetParameters=true}
   &InitializeParticle(
    Color=<rgba 2 1.5025 0.3477 1>
    Lifetime=&Lerp_Float(A=0.8, B=0.6, Alpha=$Transient.Spread)
    Sprite Rotation Mode=::NewEnumerator2
    Sprite Size Mode=::NewEnumerator1
    Uniform Sprite Size Max=35
    Uniform Sprite Size Min=25
    UsePositionOffset=false
    Write Material Random=true
   )
   &ShapeLocation(
    Box Size=&Lerp_Vector(A=<1 6 5>, B=<1 6 50>, Alpha=$Transient.Spread)
    MaskSpawnGroup=false
    Override Seed=false
    Shape Primitive=::NewEnumerator2
    Surface Only=false
    UseOwnerScale=true
   )
   &ShapeLocation(MaskSpawnGroup=false, Override Seed=false, Sphere Radius=20, UseOwnerScale=true) {enabled=false, module=ShapeLocation001}
   &CameraOffset(Camera Offset Amount=7)
  ]
  particleUpdateScript [
   &SolveForcesAndVelocity(Clamp Velocity=false, Limit Acceleration=false)
   &ParticleState()
   &ScaleSpriteSize(Uniform Curve Sprite Scale=DataInterface() {propertyValues {Curve {keys [
        {interpMode=RCIM_Cubic, tangentMode=RCTM_User, time=0, value=0}
        {interpMode=RCIM_Cubic, tangentMode=RCTM_User, time=0.5, value=1}
        {tangentMode=RCTM_User, time=1, value=0}
       ]}}})
   &SpriteRotationRate(Rotation Rate=&RandomRangeFloat(Minimum=-50, Maximum=50))
  ]
  rendererClasses [&NiagaraSpriteRendererProperties]
  renderers [{rendererClass=&NiagaraSpriteRendererProperties}]
 }
}
summary {
 systemName=NS_Emote_Guitar_Clean_Sparkle
 emitters [
  {emitterName=Sparkles, rendererClasses [&NiagaraSpriteRendererProperties]}
  {emitterName=Flare, rendererClasses [&NiagaraSpriteRendererProperties]}
 ]
 userVariables [
  User.NormalizedNotifyProgress {type=NiagaraFloat, defaultValue=0, description=""}
  User.SkeletalMesh {
   type=NiagaraDataInterfaceSkeletalMesh
   defaultValue {
    struct=NiagaraExt_VariableValue_DataInterface
    value {dataInterface=&self:NiagaraDataInterfaceSkeletalMesh_0}
   }
  }
 ]
}
